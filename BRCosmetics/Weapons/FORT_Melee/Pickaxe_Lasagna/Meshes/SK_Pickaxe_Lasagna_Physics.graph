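# Physics asset for the skeletal mesh of the Lasagna two-handed melee pickaxe cosmetic.
# A long, thin capsule covers the handle bone, and a short, near-spherical capsule covers the dyn_ball bone.
# One constraint joins the ball to the handle with both swing axes and twist set free, so its stored angular limits do not restrict rotation.
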PhysicsAsset /BRCosmetics/Weapons/FORT_Melee/Pickaxe_Lasagna/Meshes/SK_Pickaxe_Lasagna_Physics
bodies {
 Melee_TwoHanded_Handle [{shapeType=Capsule, length=50, radius=3, center=<0.0013 -1.021 18.3671>}]
 dyn_ball [{shapeType=Capsule, length=1, radius=6, rotation=<rot 90 0 0>}]
}
constraints [{
  bone1Name=dyn_ball
  bone2Name=Melee_TwoHanded_Handle
  swing1LimitDegrees=45
  swing1Motion=Free
  swing2LimitDegrees=45
  swing2Motion=Free
  twistLimitDegrees=45
  twistMotion=Free
 }]
