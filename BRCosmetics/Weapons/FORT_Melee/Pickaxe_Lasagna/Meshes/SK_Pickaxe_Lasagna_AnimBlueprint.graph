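# Animation Blueprint for the Lasagna pickaxe skeleton that continuously spins its dyn_ball bone.
# Each animation update increments a counter, wraps it at 720 and builds ball_rotator with equal roll, pitch and yaw, so speed depends on frame rate.
# The AnimGraph applies that rotator additively in bone space to the input pose.
# A RigidBody node with a physics asset override exists but is not wired into the output pose.
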AnimBlueprint /BRCosmetics/Weapons/FORT_Melee/Pickaxe_Lasagna/Meshes/SK_Pickaxe_Lasagna_AnimBlueprint
target_skeleton=&/SK_Pickaxe_Lasagna_Skeleton
graphs [
 AnimGraph {nodes [
   RigidBody() {
    bClampLinearTranslationLimitToRefPose=true
    bOverrideWorldGravity=true
    overridePhysicsAsset=&/SK_Pickaxe_Lasagna_Physics
    simulationSpace=BaseBoneSpace
    alphaBoolBlend {blendOption=Linear}
    baseBoneRef {boneName=b_AR_Root}
    componentAppliedLinearAccClamp=<3000 3000 3000>
    componentLinearAccScale=<0.5 0.5 0.5>
    componentLinearVelScale=<0.3 0.3 0.3>
    overrideWorldGravity {z=-980}
   }
   AnimGraphNode_SubInput_0 = Input Pose()
   n1 = "Transform (Modify) Bone - Bone: dyn_ball"(
    ComponentPose=Local To Component(LocalPose=@AnimGraphNode_SubInput_0)
    Alpha=1
    Translation="0,0,0"
    Rotation=Get ball_rotator() -> (ball_rotator:Rotator)
    Scale="1,1,1"
   ) {rotationMode=BMM_Additive, rotationSpace=BCS_BoneSpace, alphaBoolBlend {blendOption=Linear}, boneToModify {boneName=dyn_ball}}
   Output Pose(Result=Component To Local(ComponentPose=@n1)) {layerGroup=None}
  ]}
 EventGraph {nodes [
   Event BlueprintUpdateAnimation()
   n1 = Get Spin_Int() -> (Spin_Int:double)
   > Set Spin_Int(Spin_Int:double=float + float(A=@n1, B=1) {is=title K2Node_CommutativeAssociativeBinaryOperator}) -> (Output_Get:double)
   > n2 = Branch(Condition="float >= float"(A=@n1, B=720) {is=title K2Node_CallFunction})
   > n3 = Set Spin_Int(Spin_Int=float - float(A=Get Spin_Int(), B=720) {is=title K2Node_CallFunction})
   n4 = Set Spin_Int(execute=@n2.else, Spin_Int=float + float(A=Get Spin_Int(), B=0))
   Set ball_rotator(execute=[@n3.then, @n4.then], ball_rotator:Rotator=MakeRotator(Roll=@n4, Pitch=@n4, Yaw=@n4)) -> (Output_Get:Rotator)
   TryGetPawnOwner()
  ]}
]
variables [
 ball_rotator {type=Rotator}
 Spin_Int {type=number}
]
